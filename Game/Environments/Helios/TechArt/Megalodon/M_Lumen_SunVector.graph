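# Unlit material, apparently a Lumen lighting boost, that computes a color from how surfaces face the sun.
# A SphereMask compares the sky atmosphere light direction with the world-space pixel normal, using scalar parameters for radius and hardness.
# The inverted mask multiplies the Lumen Boost Color parameter, an over-bright green by default, so the color strengthens as normals turn away from the sun.
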Material /Game/Environments/Helios/TechArt/Megalodon/M_Lumen_SunVector
graph {
 properties {bAutomaticallySetUsageInEditor=false, shadingModel=MSM_Unlit, usedShadingModels=Unlit}
 nodes [
  n1 = SkyAtmosphereLightDirection()
  n2 = PixelNormalWS()
  n3 = ScalarParameter() {defaultValue=42, parameterName=Rad}
  n4 = ScalarParameter() {defaultValue=0.04, parameterName=asdfsd}
  Multiply(
   A=VectorParameter() {parameterName=Lumen Boost Color, defaultValue=<rgba 7.6768 15 1.6008 1>}.RGB
   B=OneMinus(None=SphereMask(A=@n1, B=@n2, Radius=@n3, Hardness=@n4) {hardnessPercent=0})
  )
 ]
}
